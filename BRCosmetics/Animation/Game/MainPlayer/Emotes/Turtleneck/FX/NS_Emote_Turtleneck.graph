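# Niagara system for the Turtleneck emote: a brief, delayed puff of pale blue-grey sprite smoke around head height.
# A single Smoke emitter spawns at a rate that falls over its short loop, placing particles on a small tilted torus offset above the origin.
# Particles burst along torus normals with downward drift, then drag and curve-driven positive Z gravity lift them.
# Sprites grow, fade, slow their spin and play SubUV animation with dynamic material parameters.
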NiagaraSystem /BRCosmetics/Animation/Game/MainPlayer/Emotes/Turtleneck/FX/NS_Emote_Turtleneck
dependencies {
 usedDataInterfaces [&NiagaraDataInterfaceCurve, &NiagaraDataInterfaceVectorCurve]
 usedDynamicInputs []
 usedModules [&SystemState]
}
emitters {Smoke {
  emitterUpdateScript [
   &EmitterState(Life Cycle Mode=::NewEnumerator1, Loop Behavior=::NewEnumerator1, Loop Delay=0.14, Loop Duration=0.2, UseLoopDelay=true)
   &SpawnRate(
    SpawnRate=&FloatFromCurve(
     FloatCurve=DataInterface() {propertyValues {Curve {keys [{arriveTangent=-1.7816, interpMode=RCIM_Cubic, leaveTangent=-1.7816, tangentMode=RCTM_User, time=0, value=1}, <at 1 0>]}}}
     CurveIndex=$Emitter.NormalizedLoopAge
     Scale Curve=800
    )
    Use Spawn Probability=false
   )
  ]
  particleSpawnScript [
   &InitializeParticle(
    Color=&MakeLinearColorFromVectorAndFloat(
     "Vector (RGB)"=&MakeVectorFromLinearColorRGB(Color=&MultiplyLinearColor(
       Linear Color=&RandomRangeLinearColor(Minimum=<rgba 0.2216 0.3466 0.5625 1>, Maximum=<rgba 0.3392 0.4778 0.6719 1>)
       Scale Factor=&MakeVector4_FromFloat(Float=1)
      ))
     "Float (Alpha)"=&UniformRangedFloat(Minimum=0.2, Maximum=0.3)
    )
    Lifetime=&UniformRangedFloat(Minimum=0.2, Maximum=1.2)
    Position Offset=&Add_Vector(A=&SimulationPosition(), B=<5 0 140>)
    Sprite Rotation Mode=::NewEnumerator2
    Sprite Size Mode=::NewEnumerator3
    Uniform Sprite Size=&Multiply_Float(A=&UniformRangedFloat(Minimum=20, Maximum=50), B=0.1)
    UsePositionOffset=true
    Write Material Random=true
   ) {module=InitializeParticle001}
   /SetVariables(Particles.InitSpriteRotRate=&UniformRangedFloat(Minimum=-3, Maximum=3)) {isSetParameters=true}
   &TorusLocation(
    Align Torus With Axis=true
    Large Radius=0.7
    MaskSpawnGroup=false
    Module.Override Randomness=false
    Normalized Angle Around Torus Axis=0.5
    Surface Position=2
    Torus Axis=<0 0.5 1>
    Torus Mode=::NewEnumerator2
    U Distribution=0.5
   )
   &AddVelocity(Velocity=&Multiply_Vector(
     A=&VectorFromFloat(Value=&RandomRangeFloat(Minimum=50, Maximum=100))
     B=$Output.TorusLocation.TorusNormal
    )) {module=AddVelocity001}
   &AddVelocity(Velocity=&RandomRangeVector(Minimum=<0 0 -40>, Maximum=<0 0 -80>)) {module=AddVelocity002}
  ]
  particleUpdateScript [
   &ParticleState()
   &SpriteRotationRate(Rotation Rate=&Multiply_Float(
     A=$Particles.InitSpriteRotRate
     B=&FloatFromCurve(FloatCurve=DataInterface() {propertyValues {Curve {keys [<at -0.0046 1.0113 RCIM_Cubic>,<at 1.0011 0.09 RCIM_Cubic>]}}}, Scale Curve=100)
    ))
   /Niagara/Modules/Update/SubUV/SubUVAnimation(Number Of Frames=64, SubUV Lookup Index=&Multiply_Float(A=$Particles.Age, B=0.66))
   &Drag(Drag=4, Use Linear Drag=true, Use Rotational Drag=false)
   &GravityForce(Gravity=&VectorFromCurve(
     VectorCurve=DataInterface() {propertyValues {XCurve {keys [<at 0 1>,<at 1 0>]}, YCurve {keys [<at 0 1>,<at 1 0>]}, ZCurve {keys [<at 0.1 0>,<at 0.8 1>]}}}
     Scale Curve=&RandomRangeVector(Minimum=<0 0 300>, Maximum=<0 0 100>)
    ))
   &SolveForcesAndVelocity(Clamp Velocity=false, Limit Acceleration=false)
   &Color(
    Color=$Particles.Initial.Color
    Scale Alpha=&FloatFromCurve(
     FloatCurve=DataInterface() {propertyValues {Curve {keys [{arriveTangent=-0.8206, interpMode=RCIM_Cubic, leaveTangent=-0.8206, tangentMode=RCTM_User, time=0, value=0.5}, <at 1.0011 0.002 RCIM_Cubic>]}}}
     Scale Curve=0.7
    )
    Scale Color=&VectorFromFloat(Value=1.5)
   )
   &ScaleSpriteSize(Scale Factor=&Vector2DFromFloat(Value=&FloatFromCurve(
      FloatCurve=DataInterface() {propertyValues {Curve {keys [
          {arriveTangent=4.1785, interpMode=RCIM_Cubic, leaveTangent=4.1785, tangentMode=RCTM_User, time=0, value=0}
          {arriveTangent=0.9995, interpMode=RCIM_Cubic, leaveTangent=0.9995, tangentMode=RCTM_User, time=0.2223, value=0.6825}
          {arriveTangent=-0.0405, interpMode=RCIM_Cubic, leaveTangent=-0.0405, tangentMode=RCTM_User, time=1, value=1}
         ]}}}
      Scale Curve=3
     )))
   &DynamicMaterialParameters(
    Index 0 Param 1=&RandomRangeFloat(Minimum=5, Maximum=1)
    Index 0 Param 2=1
    Index 0 Param 3=1
    Index 0 Param 4=1
    Module.Index0Float0Write=true
    Module.Index0Float1Write=true
    Module.Index0Float2Write=true
    Module.Index0Float3Write=true
   )
  ]
  rendererClasses [&NiagaraSpriteRendererProperties]
  renderers [{rendererClass=&NiagaraSpriteRendererProperties}]
 }}
summary {
 systemName=NS_Emote_Turtleneck
 emitters [{emitterName=Smoke, rendererClasses [&NiagaraSpriteRendererProperties]}]
 userVariables []
}
